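# Masked, lit parent material for the PaperCut vehicle wheel cosmetic, enabled for skeletal meshes.
# An offset-based infinity material function, borrowed from the ClockGear_SE wheel, processes the wheel texture with exposed Offset and GradientDarkness parameters.
# Its result is tinted dark red and masked by the T_PaperCut_SE green channel, then added to a yellow tint masked by blue.
# A separate UV-centered radial gradient is computed with no visible consumer.
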Material /VehicleCosmetics/Wheels/PaperCut/Materials/M_WheelParent_PaperCut_FX
graph {
 properties {bAutomaticallySetUsageInEditor=false, bUsedWithSkeletalMesh=true, blendMode=BLEND_Masked, usedShadingModels=Default Lit}
 nodes [
  n1 = Constant4Vector() {constant=<rgba 0.5 0 0 0>}
  n2 = TextureObjectParameter() {parameterName=Texture, samplerType=SAMPLERTYPE_LinearColor, texture=/VehicleCosmetics/Wheels/PaperCut/Textures/Wheel_PaperCut_M}
  n3 = MaterialFunctionCall!(
   Texture=@n2
   Offset=ScalarParameter() {defaultValue=0.2, parameterName=Offset}
   UVs
   RotateUVs
   GradientFalloff
   GradientDarkness=ScalarParameter() {defaultValue=0.01, parameterName=GradientDarkness}
  ) {materialFunction=/VehicleCosmetics/Wheels/ClockGear_SE/Materials/MF_Infinity_OffsetFunction}
  n4 = TextureSample() {samplerSource=SSM_Wrap_WorldGroupSettings, texture=/VehicleCosmetics/Wheels/PaperCut/Textures/T_PaperCut_SE}
  Add(A=Multiply(A=Multiply(A=@n1.RGBA, B=@n3.Result), B=@n4.G), B=Multiply(A=Constant4Vector() {constant=<rgba 1 1 0.05 0>}.RGBA, B=@n4.B))
  n5 = Add(A=TextureCoordinate(), B=Constant() {r=-0.5})
  Clamp(None=OneMinus(None=DotProduct(A=@n5, B=@n5)))
 ]
}
